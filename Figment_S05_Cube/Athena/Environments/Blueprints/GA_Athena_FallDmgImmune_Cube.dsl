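; Gameplay ability that makes a launched player immune to their next fall damage, with looping effects while airborne.
; After a short launch delay it ends if the pawn is grounded, otherwise it sets the ignore-damage flag and cue.
; It watches walking and custom (water) movement changes, polls for vehicle landings, and has a timeout failsafe.
; Cleanup clears the flag, swaps to a landing cue, removes the granting effect and ends.
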(event Ability|EventActivateAbility
  (bind _asfort_player_pawn (Utilities|Casting|CastToFortPlayerPawn (Ability|GetAvatarActorfromActorInfo))
    (:then
      (bind _output_get (Variables|Default|SetPlayerPawn _asfort_player_pawn))
      ; Give time for the player to be launched.
      (Utilities|FlowControl|Delay 0.1)
      (if (Movement|IsMovingonGround (Pawn|GetMovementComponent _output_get))
        (Ability|EndAbility)
        (else
          (Variables|/Script/FortniteGame.FortPawn|SetbIgnoreNextFallingDamage true _output_get)
          (Ability|AddGameplayCueToOwner (Variables|Default|GetLooping%20GC))
          (Utilities|FlowControl|Sequence
            (:then_0
              (Ability|Tasks|WaitMovementModeChange "MOVE_Walking"
                (:then
                  (bind _returnvalue (Utilities|Time|SetTimerbyEvent (EventDispatchers|CreateEvent :!event Custom|VehicleLandedCheck) 0.2 true))
                  (Variables|Default|SetVehicleOnGroundTimer _returnvalue)
                  (Utilities|FlowControl|RetriggerableDelay 15.0) ; Failsafe to ensure ability ends.
                )
                (:OnChange
                  _))
              (Variables|/Script/FortniteGame.FortPawn|SetbIgnoreNextFallingDamage false (Variables|Default|GetPlayerPawn))
              (Ability|RemoveGameplayCueFromOwner (Variables|Default|GetLooping%20GC))
              (Ability|AddGameplayCueToOwner (Variables|Default|GetPlayer%20Landing%20GC))
              (Utilities|Time|ClearandInvalidateTimerbyHandle (Variables|Default|GetVehicleOnGroundTimer))
              (Ability|RemoveGrantedbyEffect)
              (Ability|EndAbility))
            (:then_1
              (Ability|Tasks|WaitMovementModeChange "MOVE_Custom"
                (:OnChange
                  _)) ; Fix for fx not getting removed when landing in water (FORT-223198)
            )))))
    (:CastFailed
      (Ability|EndAbility))))

(event Ability|EventFailedToActivatePassiveAbility
  (params (ActorInfo :type Gameplay%20Ability%20Actor%20Info%20Structure))
  (|CallParentFunction))

(event Custom|VehicleLandedCheck
  (bind _playerpawn (Variables|Default|GetPlayerPawn))
  (bind _returnvalue (Vehicle|GetVehicleInterface _playerpawn))
  ;[1 1 1] So Bouncer FX stay on a player if they exit a vehicle before landing. Fix FORT-145705
  (Utilities|IsValid (Utilities|ToObject(Interface) _returnvalue)
    (:Is%20Valid
      (if (and (Class|FortVehicleInterface|IsTouchingGround _returnvalue) (Vehicle|IsinVehicle _playerpawn))
        (Utilities|Time|ClearandInvalidateTimerbyHandle (Variables|Default|GetVehicleOnGroundTimer))
        (Variables|/Script/FortniteGame.FortPawn|SetbIgnoreNextFallingDamage false (Variables|Default|GetPlayerPawn))
        (Ability|RemoveGameplayCueFromOwner (Variables|Default|GetLooping%20GC))
        (Ability|AddGameplayCueToOwner (Variables|Default|GetPlayer%20Landing%20GC))
        (Utilities|Time|ClearandInvalidateTimerbyHandle (Variables|Default|GetVehicleOnGroundTimer))
        (Ability|RemoveGrantedbyEffect)
        (Ability|EndAbility)))))
